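# Idle Niagara effect for the Cactus Rocker guitar cosmetic, drawing five persistent mesh particles with audio-reactive opacity.
# One burst is partitioned so array lookups give each mesh its scale, offset, tilt and material parameters; socket-based skeletal placement is disabled.
# Alpha follows event audio amplitude, plus kick and snare stems in gameplay, with separate front-end and gameplay tuning.
# A smoothed particle position is packed into a dynamic vector parameter.
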NiagaraSystem /SparksCosmetics/Instrument/Guitar/CactusRocker/Effects/NS_Sparks_Guitar_Cactus_Rocker_Idle
dependencies {
 usedDataInterfaces [
  &NiagaraDataInterfaceArrayFloat3
  &NiagaraDataInterfaceArrayFloat
  &NiagaraDataInterfaceSkeletalMesh
  &NiagaraDataInterfaceCurve
 ]
 usedDynamicInputs []
 usedModules [&SystemState]
}
emitters {Idle_FX {
  emitterUpdateScript [
   &EmitterState(Life Cycle Mode=::NewEnumerator1, Loop Behavior=::NewEnumerator1, Loop Duration Mode=::NewEnumerator1, UseLoopDelay=false)
   &SpawnBurst_Instantaneous(Spawn Count=5, Use Spawn Probability=false, UseLoopCountLimit=false)
  ]
  particleSpawnScript [
   &PartitionParticles(Number Of Partitions=5)
   &FN_SetSpawnGroup(Number of Spawn Groups=5) {enabled=false}
   &InitializeParticle(
    Mesh Scale=&SelectVectorFromArray(
     Direct Array Index=$Particles.PartitionParticles.Partition
     Vector Selection Array=DataInterface() {propertyValues {FloatData [<0.5 1 1.25>, <0.5 1.36 2.25>, <0.7 0.6 0.8>, <0.4 0.575 1.48>, <0.1 0.6 1.3>]}}
    )
    Mesh Scale Mode=::NewEnumerator4
    Position Offset=&SelectVectorFromArray(
     Direct Array Index=$Particles.PartitionParticles.Partition
     Vector Selection Array=DataInterface() {propertyValues {FloatData [<4 -1.48 0>, <-3 -1.5 -8>, <-13 -3 -21>, <10 -3 -27>, <-2 0 71>]}}
    )
    UsePositionOffset=true
    Write Material Random=true
   )
   &InitialMeshOrientation(Mesh Orientation Mode=::NewEnumerator7, Rotation=<0 0.25 0>, Use Rotation Vector=true)
   &InitialMeshOrientation(
    Mesh Orientation Mode=::NewEnumerator7
    Rotation=&SelectVectorFromArray(
     Direct Array Index=$Particles.PartitionParticles.Partition
     Vector Selection Array=DataInterface() {propertyValues {FloatData [<0 0 0>, <0 0 0.1>, <0 0 0.1>, <0 0 0.05>, <0 0 0>]}}
    )
    Use Rotation Vector=true
   ) {module=InitialMeshOrientation001}
   &InitialMeshOrientation(Mesh Orientation Mode=::NewEnumerator7, Rotation=<-0.25 0 0>, Use Rotation Vector=true) {module=InitialMeshOrientation002}
  ]
  particleUpdateScript [
   &DynamicMaterialParameters(
    Index 0 Param 1=&RandomRangeFloat(Minimum=0, Maximum=0.5)
    Index 0 Param 2=&RandomRangeFloat(Minimum=1, Maximum=0.5)
    Index 0 Param 3=&self:ScratchDynamicInput(
     Front End:NiagaraFloat=&SelectFloatFromArray(
      Direct Array Index=$Particles.Initial.PartitionParticles.Partition
      Float Selection Array=DataInterface() {propertyValues {FloatData [12,10,14,18,10]}}
     )
     Gameplay:NiagaraFloat=&SelectFloatFromArray(
      Direct Array Index=$Particles.Initial.PartitionParticles.Partition
      Float Selection Array=DataInterface() {propertyValues {FloatData [12,10,14,18,10]}}
     )
     IsFrontEnd:NiagaraBool=$User.IsFrontEnd
    )
    Index 0 Param 4=&SelectFloatFromArray(
     Direct Array Index=$Particles.Initial.PartitionParticles.Partition
     Float Selection Array=DataInterface() {propertyValues {FloatData [30,40,20,20,20]}}
    )
    Index 2 Param 1=&FN_FrontEndFloatModifier(Gameplay=0.3, Front End=0.5)
    Module.Param0WriteEnabled=true
    Module.Param0WriteEnabled2=true
    Module.Param1WriteEnabled=true
    Module.Param1WriteEnabled2=true
    Module.Param2WriteEnabled=true
    Module.Param2WriteEnabled2=true
    Module.Param3WriteEnabled=true
    Module.Param3WriteEnabled2=true
    Write Parameter Index 2=true
   )
   &ParticleState(Kill Particles When Lifetime Has Elapsed=false, Let Infinitely Lived Particles Die When Emitter Deactivates=true)
   &InterpolateOverTime(
    Rate Of Change=8
    Target Vector Value:Vector3f=&ConvertPositionToVector(Input Position=$Particles.Position, Convert To Absolute Large World Vector=false)
    Variable Type=::NewEnumerator2
   )
   /SetVariables(Particles.Interpolate Dynamic=&MakeFloat4FromVectorAndFloat("Vector (RGB)"=$StackContext.InterpolateOverTime.InterpolateOverTime_V3_.Moving Average, "Float (Alpha)"=0.55)) {isSetParameters=true}
   &SkeletalMeshLocation(
    Bone / Socket Index=$Particles.SpawnGroup
    MaskSpawnGroup=false
    Mesh Sampling Type=::NewEnumerator5
    Sampled Position Offset=&SelectVectorFromArray(
     Direct Array Index=$Particles.SpawnGroup
     Vector Selection Array=DataInterface() {propertyValues {FloatData [<0 0 4>, <0 0 -15>, <-3.5 0 -2>, <-7 0 -1.5>]}}
    )
    Skeletal Mesh=DataInterface() {propertyValues {
      DefaultMesh=/SparksCosmetics/Instrument/Guitar/CactusRocker/Meshes/CactusRocker_Guitar
      ExcludeBoneName=root
      OptionalComponentClass=null
      PreviewMesh=&CactusRocker_Guitar
      SkinningMode=SkinOnTheFly
      SoftSourceActor=null
      SourceMode=Source
      UvSetIndex=0
      WholeMeshLOD=-1
      bExcludeBone=false
      bReadDeformedGeometry=true
      bRequireCurrentFrameData=true
      ComponentTags []
      FilteredBones []
      FilteredSockets [Idle_FX,Idle_FX,Idle_FX_b,Idle_FX_c,Idle_FX_d]
      MeshUserParameter {parameter {name=None, typeDefHandle {registeredTypeIndex=1}, varData []}}
      SamplingRegions []
     }}
    Socket Sampling Mode=::NewEnumerator5
   ) {enabled=false}
   &ScaleColor(
    Scale Alpha=&FN_FrontEndFloatModifier(
     Gameplay=&Lerp_Float(
      A=&Add_Float(
       A=0.055
       B=&Multiply_Float(
        A=0.1
        B=&Max_Float(A:NiagaraFloat=$NPC.NPC_Audio_Reactive_Parameters.MusicStem_Snare, B:NiagaraFloat=$NPC.NPC_Audio_Reactive_Parameters.MusicStem_Kick)
       )
      )
      B=0.24
      Alpha=$NPC.FortniteNPC.EventAudio_Amplitude
     )
     Front End=&Lerp_Float(
      A=0.12
      B=0.6
      Alpha=&ClampFloat(Float=&FloatFromCurve(
        FloatCurve=DataInterface() {propertyValues {Curve {keys [{interpMode=RCIM_Cubic, tangentMode=RCTM_Break, time=0, value=-2}, {interpMode=RCIM_Cubic, tangentMode=RCTM_Break, time=1, value=5}]}}}
        CurveIndex=$NPC.FortniteNPC.EventAudio_Amplitude
       ))
     )
    )
    Scale RGB=&Multiply_VectorByFloat(Vector=<1 1 1>, Float=&FN_FrontEndFloatModifier(Gameplay=0.8, Front End=1))
    ScaleA=true
    ScaleRGB=true
   )
   &ScaleMeshSize()
  ]
  rendererClasses [&NiagaraMeshRendererProperties]
  renderers [{rendererClass=&NiagaraMeshRendererProperties, rendererIndex=0}]
 }}
summary {
 systemName=NS_Sparks_Guitar_Cactus_Rocker_Idle
 emitters [{emitterName=Idle_FX, rendererClasses [&NiagaraMeshRendererProperties]}]
 userVariables [
  User.bIsFrontEnd {type=NiagaraBool}
  User.bIsFrontEndPreview {type=NiagaraBool}
  User.IsFrontEnd {type=NiagaraBool}
 ]
}
